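# Conversation database for the DinoHunter NPC, entered through its own conversation entry tag.
# The entry leads to a speech node with a default warning about respecting the island, plus priority-curved contextual lines for a specific CubeNinja outfit or a first-ever conversation.
# The speech then branches through link nodes into separate BasicServices, Season 16 Spire DinoHunter, and NPC library quest conversations.
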ConversationDatabase /NPCLibrary/NPCs/DinoHunter/DinoHunter/NPCLibrary_Conversation_DinoHunter
entry_points [{tag=Conversation.Entry.DinoHunter, destinations [#1]}]
nodes [
 ConversationEntryPointNode_7 {
  id=#1
  class=ConversationEntryPointNode
  outputs [#2]
  properties {entryTag {tagName=Conversation.Entry.DinoHunter}}
 }
 Speech_C_4 {
  id=#2
  class=Speech_C
  outputs [#3, #4, #5]
  properties {
   nodeName=""
   generalConfig {
    defaultMessage="Respect the island and we shouldn't have any <enemy>problems</> with each other."
    contextualMessages [
     {
      message="I've got my eye on you. "
      contextRequirements [{requirement {
         class=FortControllerRequirement_HasCID
         properties {allowedCIds [/BRCosmetics/Athena/Items/Cosmetics/Characters/CID_A_041_Athena_Commando_M_CubeNinja]}
        }}]
      priority {curve {rowName=Tandem.Speech.Priority.OutfitReaction}}
     }
     {
      message=Look what the Spire dragged in this time.
      contextRequirements [{requirement {class=FortControllerRequirement_IsFirstEverConversationWithNPC}}]
      priority {curve {rowName=Tandem.Speech.Priority.FirstEverInteraction}}
     }
    ]
   }
  }
 }
 ConversationLinkNode_1 {id=#5, class=ConversationLinkNode, properties {remoteEntryTag {tagName=Conversation.Entry.NPCLibrary.Quests}}}
 ConversationLinkNode_2 {
  id=#4
  class=ConversationLinkNode
  properties {remoteEntryTag {tagName=Conversation.Entry.Season16.TheSpire.DinoHunter}}
 }
 ConversationLinkNode_0 {id=#3, class=ConversationLinkNode, properties {remoteEntryTag {tagName=Conversation.Entry.BasicServices}}}
]
